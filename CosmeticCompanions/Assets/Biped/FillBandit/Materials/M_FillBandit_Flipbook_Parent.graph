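# Parent cloth-shaded anime material for the skeletal-mesh FillBandit companion, adding an animated flipbook face.
# Chains the shared 2023 anime Pre, Calculations and Post functions, overriding world normals through a companion normal-sphere function.
# Flipbook face color blends into lit and shaded colors by face alpha, then passes through pet eye animation masked by high red values.
# The face function's world position offset adds to the shader's, and temporal responsiveness is output separately.
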Material /CosmeticCompanions/Assets/Biped/FillBandit/Materials/M_FillBandit_Flipbook_Parent
graph {
 properties {
  bAutomaticallySetUsageInEditor=false
  bUseMaterialAttributes=true
  bUsedWithMeshDeformer=true
  bUsedWithMorphTargets=true
  bUsedWithSkeletalMesh=true
  pixelDepthOffsetMode=PDOM_Legacy
  shadingModel=MSM_Cloth
  usedShadingModels=Cloth
 }
 nodes [
  n1 = MaterialFunctionCall!(CustomUV) {materialFunction=/Game/Global/Materials/Character_Parent_2023/Anime_Parent_2023/MaterialFunctions/MFC_AnimeShader_2023_Pre}
  n2 = MaterialFunctionCall!(
   CustomLightVector
   OverrideNormal_WS=MaterialFunctionCall!(Normal=@n1.Normal) {materialFunction=/CosmeticCompanions/Materials/MaterialFunctions/MF_CompanionAnime_NormalSphere}.Normal
   STT=@n1.STT
   Normal_Map=@n1.STT
   DFL=@n1.STT
   FX=@n1.STT
  ) {materialFunction=/Game/Global/Materials/Character_Parent_2023/Anime_Parent_2023/MaterialFunctions/MFC_AnimeShader_2023_Calculations}
  n3 = MaterialFunctionCall!(Material, Brow UV, Eye UV, Mouth UV) {materialFunction=/CosmeticCompanions/Materials/MaterialFunctions/MF_FlipbookFace_Companion}
  n4 = NamedRerouteDeclaration(None=@n3.FaceAlpha) {name=FlipbookFaceAlpha, nodeColor=<rgba 0.8281 0 1 1>}
  n5 = ScalarParameter() {defaultValue=1, parameterName=FlipbookFaceAnimeShadedValue}
  n6 = Constant()
  n7 = MaterialFunctionCall!(
   Color Lit=LinearInterpolate(A=@n1.Color_Lit, B=@n3.FaceColor, Alpha=@n4)
   Color Shaded=LinearInterpolate(A=@n1.Color_Shaded, B=Multiply(A=@n3.FaceColor, B=@n5), Alpha=@n4)
   Use UV2
   Mask=If(A=ComponentMask(None=@n3.FaceColor) {r=true}, "A > B"=Constant() {r=1}, "A == B"=@n6, "A < B"=@n6) {constB=0.9}
  ) {materialFunction=&/MF_AnimeEyeAnimation_FillBandit_Pet}
  n8 = GetMaterialAttributes(None=MaterialFunctionCall!(
    Shading_Gradient=@n2.Shading_Gradient
    Threshold_Gradient=@n2.Shading_Gradient
    Specular_Gradient=@n2.Shading_Gradient
    Sheen_Mask=@n2.Shading_Gradient
    Thickness=@n2.Shading_Gradient
    DFL_Mask=@n2.Shading_Gradient
    FX=@n2.Shading_Gradient
    Metallic_Gradient=@n2.Shading_Gradient
    RimLight_Gradient=@n2.Shading_Gradient
    NormalBuffer=@n2.Shading_Gradient
    --------
    CINE_LightSources=@n2.Shading_Gradient
    CINE_LightColor=@n2.Shading_Gradient
    ---------
    Color_Lit=@n7.Lit
    Color_Shaded=@n7.Lit
    SRM=@n1.SRM
    Emissive=@n1.SRM
    DFL_Color=@n1.SRM
    Specular_Color=@n1.SRM
    RimLight_Color=@n1.SRM
   ) {materialFunction=/Game/Global/Materials/Character_Parent_2023/Anime_Parent_2023/MaterialFunctions/MFC_AnimeShader_2023_Post}.Result) {attributeGetTypes [#1]}
  SetMaterialAttributes!(
   MaterialAttributes=@n8.MaterialAttributes
   World Position Offset=Add(
    A=@n8.World Position Offset
    B=GetMaterialAttributes(None=@n3.Material) {attributeGetTypes [#1]}.World Position Offset
   )
  ) {attributeSetTypes [#1]}
  TemporalResponsivenessOutput!(None=LinearInterpolate(B=ScalarParameter() {parameterName=Temporal Responsiveness}, Alpha=NamedRerouteUsage()) {constA=0.4})
 ]
}
