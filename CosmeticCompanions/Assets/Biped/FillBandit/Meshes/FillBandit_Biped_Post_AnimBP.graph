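# Post-process animation blueprint for the FillBandit biped companion, judging by its name, that adjusts arm proportions on the incoming pose.
# A component-space chain of Transform (Modify) Bone nodes offsets both upper arms oppositely along X, shrinks the hands and thins the lower arms.
# All adjustments are additive, fixed and symmetrical.
# A Pose Asset node bound to a template-default variable is left disconnected from the output, and the event graph is empty.
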AnimBlueprint /CosmeticCompanions/Assets/Biped/FillBandit/Meshes/FillBandit_Biped_Post_AnimBP
target_skeleton=&/FillBandit_Biped_Skeleton
graphs [
 AnimGraph {nodes [
   Pose!(PoseAsset:Pose Asset Object Reference=Get PoseAsset() -> (PoseAsset:Pose Asset Object Reference), SourcePose:Pose Link Structure) -> (Pose:Pose Link Structure) {is=title AnimGraphNode_PoseBlendNode, blendOption=Linear}
   n1 = Input Pose()
   n2 = "Transform (Modify) Bone - Bone: upperarm_r"(
    ComponentPose="Transform (Modify) Bone - Bone: upperarm_l"(ComponentPose=Local To Component(LocalPose=@n1), Alpha=1, Translation="1,0,0", Rotation="0,0,0", Scale="1,1,1") {translationMode=BMM_Additive, boneToModify {boneName=upperarm_l}, translation {x=1}}
    Alpha=1
    Translation="-1,0,0"
    Rotation="0,0,0"
    Scale="1,1,1"
   ) {translationMode=BMM_Additive, boneToModify {boneName=upperarm_r}, translation {x=-1}}
   n3 = "Transform (Modify) Bone - Bone: lowerarm_l"(
    ComponentPose="Transform (Modify) Bone - Bone: hand_r"(
     ComponentPose="Transform (Modify) Bone - Bone: hand_l"(ComponentPose=@n2, Alpha=1, Translation="0,0,0", Rotation="0,0,0", Scale="0.8,0.8,0.8") {scaleMode=BMM_Additive, boneToModify {boneName=hand_l}, scale=<0.8 0.8 0.8>}
     Alpha=1
     Translation="0,0,0"
     Rotation="0,0,0"
     Scale="0.8,0.8,0.8"
    ) {scaleMode=BMM_Additive, boneToModify {boneName=hand_r}, scale=<0.8 0.8 0.8>}
    Alpha=1
    Translation="0,0,0"
    Rotation="0,0,0"
    Scale="1,0.8,0.8"
   ) {scaleMode=BMM_Additive, boneToModify {boneName=lowerarm_l}, scale {y=0.8, z=0.8}}
   Output Pose(Result=Component To Local(ComponentPose="Transform (Modify) Bone - Bone: lowerarm_r"(ComponentPose=@n3, Alpha=1, Translation="0,0,0", Rotation="0,0,0", Scale="1,0.8,0.8") {scaleMode=BMM_Additive, boneToModify {boneName=lowerarm_r}, scale {y=0.8, z=0.8}}))
  ]}
 EventGraph {links [], nodes []}
]
variables [PoseAsset {type=PoseAsset, default=/CosmeticCompanions/Automation_Template/companion_pose_asset_template}]
